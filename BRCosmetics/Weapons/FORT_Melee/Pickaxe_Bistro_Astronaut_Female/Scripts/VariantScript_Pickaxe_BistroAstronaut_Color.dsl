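; Variant script for the female Bistro Astronaut pickaxe that picks its material variant from the player's cosmetic loadout.
; When the matching Bistro Astronaut character is equipped, its Material channel drives the pickaxe, translating Stage1 and Stage2 tags to Mat1 and Mat2, with an empty result otherwise.
; Without that character, it returns the pickaxe's own active Material variant.
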(fn DetermineLoadoutVariantSelection
  (params (Loadout :type Cosmetic%20Loadout%20Structure%20%28by%20ref%29))
  (if (Class|FortCosmeticFunctionLibrary|IsItemInLoadout Loadout "/BRCosmetics/Athena/Items/Cosmetics/Characters/CID_A_227_Athena_Commando_F_BistroAstronaut_JJLK5.CID_A_227_Athena_Commando_F_BistroAstronaut_JJLK5")
    (switch Utilities|FlowControl|Switch|SwitchonGameplayTag (Class|FortCosmeticFunctionLibrary|GetActiveVariantForSoftItemAndChannel Loadout "/BRCosmetics/Athena/Items/Cosmetics/Characters/CID_A_227_Athena_Commando_F_BistroAstronaut_JJLK5.CID_A_227_Athena_Commando_F_BistroAstronaut_JJLK5" "(TagName=\"Cosmetics.Variant.Channel.Material\")")
      (:Default
        (return 0))
      (:Cosmetics.Variant.Property.Stage1
        (return "(TagName=\"Cosmetics.Variant.Property.Mat1\")"))
      (:Cosmetics.Variant.Property.Stage2
        (return "(TagName=\"Cosmetics.Variant.Property.Mat2\")")))
    (else
      (return (Class|FortCosmeticFunctionLibrary|GetActiveVariantForSoftItemAndChannel Loadout "/BRCosmetics/Athena/Items/Cosmetics/Pickaxes/Pickaxe_ID_682_BistroAstronautFemale_A3MD2.Pickaxe_ID_682_BistroAstronautFemale_A3MD2" "(TagName=\"Cosmetics.Variant.Channel.Material\")")))))
